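# Test masked, two-sided lit material for a string of lights, with glowing orange bulbs and animated sway.
# SplineThicken widens the mesh, and its offset is added to vector-noise motion weighted by UV height and vertex red, with per-instance random speed and scale.
# A power-sharpened channel of the tiled light-string texture masks the glow, which pulses through LinearSine with phase and period varied by vertex red.
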Material /Game/Effects/Fort_Effects/Mesh/Hera/M_Lights_String_Test
graph {
 properties {
  bAutomaticallySetUsageInEditor=false
  blendMode=BLEND_Masked
  translucencyLightingMode=TLM_SurfacePerPixelLighting
  twoSided=true
  usedShadingModels=Default Lit
 }
 nodes [
  n1 = Multiply(
   A=TextureCoordinate() {coordinateIndex=1, vTiling=5}
   B=AppendVector(A=ScalarParameter() {defaultValue=1, parameterName=X}, B=ScalarParameter() {defaultValue=4, parameterName=Y})
  )
  n2 = ComponentMask(None=TextureCoordinate()) {g=true}
  n3 = ObjectPositionWS()
  n4 = Subtract(A=WorldPosition().XYZ, B=@n3)
  n5 = ScalarParameter() {defaultValue=1, group=Movement, parameterName=Curl_Speed}
  n6 = Multiply(A=LinearInterpolate(A=@n5, B=Multiply(A=@n5) {constA=0.65, constB=0.65}, Alpha=PerInstanceRandom()), B=Time())
  n7 = ScalarParameter() {defaultValue=1, group=Movement, parameterName=Curl_Scale}
  n8 = Multiply(A=@n7) {constA=0.65, constB=0.65}
  n9 = PerInstanceRandom()
  n10 = VectorNoise(World Position=Divide(A=Subtract(A=@n4, B=@n6), B=LinearInterpolate(A=@n7, B=@n8, Alpha=@n9))) {noiseFunction=VNF_VectorALU}
  n11 = Multiply(A=ScalarParameter() {defaultValue=1, group=Movement, parameterName=Curl_Intensity}, B=LinearInterpolate(Alpha=VertexColor().R) {constA=1})
  Add(
   A=MaterialFunctionCall(
    WidthBase=ScalarParameter() {defaultValue=50, parameterName="WidthBase (S)"}
    WidthTip=ScalarParameter() {defaultValue=50, parameterName="WidthTip (S)"}
    UVs for Projection=@n1
   ) {materialFunction=&SplineThicken}.WorldPositionOffset
   B=Multiply(A=@n2, B=Multiply(A=@n10, B=@n11))
  )
  Constant3Vector() {constant=<rgba 0.1771 0.1771 0.1771 0>}
  Constant() {r=0.1}
  n12 = Time()
  n13 = LinearInterpolate(Alpha=VertexColor().R) {constA=0.25}
  n14 = Multiply(A=MaterialFunctionCall(Value=Add(A=@n12, B=@n13), Period=@n13) {materialFunction=&LinearSine}.Linear Sine)
  n15 = Power(
   Base=TextureSample(UVs=@n1) {texture=&/T_Light_String_D_02}.G
   Exp=ScalarParameter() {parameterName=Power, sortPriority=8}
  ) {constExponent=32}
  Multiply(
   A=Multiply(A=LinearInterpolate(Alpha=@n14) {constA=0.1})
   B=Multiply(A=@n15, B=Constant3Vector() {constant=<rgba 150 57.2121 0 1>}) {constB=150}
  )
  Multiply(A=@n2, B=@n2)
  n16 = TextureCoordinate()
  n17 = Constant2Vector() {g=0.5, r=0.5}
  n18 = ScalarParameter() {defaultValue=0.25}
  n19 = OneMinus(None=TextureSample(UVs=MaterialFunctionCall(UVs=@n16, Rotation Center=@n17, "Rotation Angle (0-1)"=@n18) {materialFunction=&CustomRotator}.Rotated Values) {texture=/Game/Global/Textures/Default/Masks_and_Gradients/T_Water_Top_Ramp}.G)
  Multiply(A=@n19, B=@n19)
  Multiply(A=VertexColor(), B=@n15)
  PerInstanceRandom()
  ScalarParameter() {parameterName=s}
  StaticSwitchParameter(True=ActorPositionWS(), False=@n3) {defaultValue=true, parameterName=ActorPositionWPO}
  TextureSample(UVs=@n1) {samplerType=SAMPLERTYPE_Normal, texture=&/T_Light_String_N_02}
 ]
}
